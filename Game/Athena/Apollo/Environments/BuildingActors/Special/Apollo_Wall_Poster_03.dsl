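; Apollo wall poster actor that reveals Buffet teaser or announce art for calendar events and recolors itself from item wraps.
; Once the match is ready, two event checks set visibility flags, whose OnReps async-load the matching material and unhide the poster.
; OnRep_UpdateMaterials runs MaterialSwap, skipped when Visible_02 is set.
; It feeds game time and four colors from the interacting player's first wrap, or random map picks, into a dynamic material.
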(event Custom|MaterialSwap
  (bind _interacting_pawn (Variables|Default|GetInteracting%20Pawn))
  (bind _returnvalue (Fortnite|GetFortPlayerControllerfromActor _interacting_pawn))
  (bind _returnvalue_2 (CosmeticLoadout|GetItemWrapsfromController _returnvalue))
  (bind _output (Utilities|Array|Get(acopy) _returnvalue_2))
  (bind _returnvalue_3 (AssetManager|GetPrimaryAssetIdfromObject _output))
  (bind _returnvalue_4 (AssetManager|IsValidPrimaryAssetId _returnvalue_3))
  (bind _wrapcolors (Variables|Default|GetWrapColors))
  (bind _wrapcolors_2 (Variables|/Game/Athena/Prototype/wrapworld/WrapIDColorsList.WrapIDColorsList_C|GetWrapColors _wrapcolors))
  (bind (_value _returnvalue_5) (Utilities|Map|Find _wrapcolors_2 _returnvalue_3))
  (bind _returnvalue_6 (and _returnvalue_5 _returnvalue_4))
  (if (not (Variables|Default|GetVisible_02))
    (bind _keys (Utilities|Map|Keys _wrapcolors_2))
    (if (or (.ReturnValue (Utilities|Map|Find _wrapcolors_2 (Utilities|Array|RandomArrayItem _keys))) _returnvalue_6)
      (bind _returnvalue_7 (Rendering|Material|CreateDynamicMaterialInstance (Variables|Default|GetPoster) 0 0 "None" :!class /Script/Engine.PrimitiveComponent))
      (Rendering|Material|SetScalarParameterValue _returnvalue_7 "GameTimeFromBP" (Utilities|Time|GetGameTimeinSeconds) :!class /Script/Engine.MaterialInstanceDynamic)
      (Rendering|Material|SetVectorParameterValue _returnvalue_7 "WrapColor0" (Math|Conversions|ToLinearColor(Vector) (Utilities|Struct|BreakWrapColors (select _returnvalue_6 _value (Utilities|Map|Find _wrapcolors_2 (Utilities|Array|RandomArrayItem _keys))))) :!class /Script/Engine.MaterialInstanceDynamic)
      (Rendering|Material|SetVectorParameterValue _returnvalue_7 "WrapColor1" (Math|Conversions|ToLinearColor(Vector) (.Sample1_7_D2115D9747204E3A70616A8424196148 (Utilities|Struct|BreakWrapColors (select _returnvalue_6 _value (Utilities|Map|Find _wrapcolors_2 (Utilities|Array|RandomArrayItem _keys)))))) :!class /Script/Engine.MaterialInstanceDynamic)
      (Rendering|Material|SetVectorParameterValue _returnvalue_7 "WrapColor2" (Math|Conversions|ToLinearColor(Vector) (.Sample2_8_0035447C4E03CF63E4039FA2156DE54A (Utilities|Struct|BreakWrapColors (select _returnvalue_6 _value (Utilities|Map|Find _wrapcolors_2 (Utilities|Array|RandomArrayItem _keys)))))) :!class /Script/Engine.MaterialInstanceDynamic)
      (Rendering|Material|SetVectorParameterValue _returnvalue_7 "WrapColor3" (Math|Conversions|ToLinearColor(Vector) (.Sample3_9_BD6CD06C40AE128CD393D38A993F9798 (Utilities|Struct|BreakWrapColors (select _returnvalue_6 _value (Utilities|Map|Find _wrapcolors_2 (Utilities|Array|RandomArrayItem _keys)))))) :!class /Script/Engine.MaterialInstanceDynamic))))

(event EventBeginPlay
  (Class|FortAsyncActionAuthorityMatchReady|MatchReadyAsync
    (:OnReady
      (Utilities|FlowControl|Sequence
        (:then_0
          (bind _biseventactive (CalendarEvents|GetCalendarEventInformation (Variables|Default|GetEvent%20To%20Watch_01)))
          (if _biseventactive
            (Variables|Default|SetVisible_01 true)))
        (:then_1
          (bind _biseventactive_2 (CalendarEvents|GetCalendarEventInformation (Variables|Default|GetEvent%20To%20Watch_02)))
          (if _biseventactive_2
            (Variables|Default|SetVisible_02 true)))))))

(event Custom|SetMaterial
  (bind _object (Utilities|AsyncLoadAsset (Variables|Default|GetBuffetTeaserMaterial))
    (:Completed
      (bind _asmaterial_instance (Utilities|Casting|CastToMaterialInstance _object))
      (Rendering|Material|SetMaterial (Variables|Default|GetPoster) 0 _asmaterial_instance)
      (Rendering|SetActorHiddenInGame false)
      (Development|SetHiddeninGame (Variables|Default|GetPoster) false false))))

(event Custom|SetMaterial_02
  (bind _object (Utilities|AsyncLoadAsset (Variables|Default|GetBuffetAnnounceMaterial))
    (:Completed
      (bind _asmaterial_instance (Utilities|Casting|CastToMaterialInstance _object))
      (Rendering|Material|SetMaterial (Variables|Default|GetPoster) 0 _asmaterial_instance)
      (Rendering|SetActorHiddenInGame false)
      (Development|SetHiddeninGame (Variables|Default|GetPoster) false false))))

(fn OnRep_UpdateMaterials
  (CallFunction|MaterialSwap))

(fn OnRep_Visible
  (CallFunction|SetMaterial))

(fn OnRep_Visible_02
  (CallFunction|SetMaterial02))

(fn ConstructionScript)
